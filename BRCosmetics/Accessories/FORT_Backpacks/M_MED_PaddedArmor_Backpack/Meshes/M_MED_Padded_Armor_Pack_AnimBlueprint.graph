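# Animation Blueprint for the padded armor backpack cosmetic that adds secondary physics motion to its left and right zipper bones.
# The local-space reference pose passes through two chained Anim Dynamics nodes, one per zipper, before reaching the output pose.
# Each simulation uses angular limits, heavy damping, a sim-space gravity override, and scaled, clamped component motion.
# The event graph holds only unconnected default nodes, with no logic or variables.
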AnimBlueprint /BRCosmetics/Accessories/FORT_Backpacks/M_MED_PaddedArmor_Backpack/Meshes/M_MED_Padded_Armor_Pack_AnimBlueprint
target_skeleton=&/M_MED_Padded_Armor_Pack_Skeleton
graphs [
 AnimGraph {nodes [
   n1 = Anim Dynamics - Bone: dyn_l_zipper(ComponentPose=Local To Component(LocalPose=Local Space Ref Pose()), Alpha=1, ExternalForce="0,0,0") {
    angularDampingOverride=0.98
    bGravityOverrideInSimSpace=true
    bOverrideAngularDamping=true
    bUseGravityOverride=true
    alphaBoolBlend {blendOption=Linear}
    boundBone {boneName=dyn_l_zipper}
    componentAppliedLinearAccClamp=<1500 1500 1500>
    componentLinearAccScale=<0.5 0.5 0.3>
    componentLinearVelScale=<0.3 0.3 0.3>
    gravityOverride {z=-980}
    physicsBodyDefinitions [{
      sphereCollisionRadius=0
      boundBone {boneName=dyn_l_zipper}
      boxExtents=<4 5 5>
      constraintSetup {angularLimitsMax=<0 0 90>, angularLimitsMin=<0 -90 -90>, angularTarget=<0 0 0>}
      localJointOffset=<1.2 0 0>
     }]
   }
   n2 = Anim Dynamics - Bone: dyn_r_zipper(ComponentPose=@n1, Alpha=1, ExternalForce="0,0,0") {
    angularDampingOverride=0.98
    bGravityOverrideInSimSpace=true
    bOverrideAngularDamping=true
    bUseGravityOverride=true
    alphaBoolBlend {blendOption=Linear}
    boundBone {boneName=dyn_r_zipper}
    componentAppliedLinearAccClamp=<1500 1500 1500>
    componentLinearAccScale=<0.5 0.5 0.3>
    componentLinearVelScale=<0.2 0.2 0.2>
    gravityOverride {z=-980}
    physicsBodyDefinitions [{
      sphereCollisionRadius=0
      boundBone {boneName=dyn_r_zipper}
      boxExtents=<4 5 5>
      constraintSetup {angularLimitsMax=<0 0 90>, angularLimitsMin=<0 -90 -90>, angularTarget=<0 0 0>}
      localJointOffset=<1 0 0>
     }]
   }
   Output Pose(Result=Component To Local(ComponentPose=@n2)) {layerGroup=None}
  ]}
 EventGraph {nodes [
   TryGetPawnOwner()
   Event BlueprintUpdateAnimation()
  ]}
]
variables []
